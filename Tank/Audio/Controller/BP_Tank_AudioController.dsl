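; BP_Tank_AudioController is the audio controller Blueprint for the drivable tank, driving its engine, track, weapon, damage and event sounds.
; Unless nativization is enabled, each update while awake feeds speed, tread power, turning, turret and cannon rotation deltas, surface, water, air and fuel into layered audio parameters.
; Gear-sign changes and turning trigger cross-fading accel, decel and turn one-shots.
; Vehicle events handle horn, boost, collisions, damage stages, overheating and destruction pass-bys.
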(event VehicleAudio|EventOnInit
  (Utilities|Struct|SetmembersinTankAudioUpdateContext (Variables|FortTankVehicleAudioController|GetContext) (Variables|Default|GetTracksLeftAudio) (Variables|Default|GetTracksRightAudio)))

(event Custom|On%20Tick%20Audio)

(event VehicleAudio|EventOnUpdate
  (params (DeltaSeconds :type Float%20%28single-precision%29))
  (bind _cannonrotation (Variables|Weapons|GetCannonRotation))
  (bind _vehicle (Variables|Pawn|GetVehicle))
  (bind _skeletalmesh (Variables|/Script/FortniteGame.FortAthenaVehicle|GetSkeletalMesh _vehicle))
  (bind _returnvalue (Transformation|GetSocketRotation _skeletalmesh "cannon_main_01"))
  (bind (_returnvalue_x _returnvalue_y) (Math|Vector2D|BreakVector2D (Weapons|GetDeltaRotation _returnvalue _cannonrotation)))
  (bind _vehicle_2 (Variables|Pawn|GetVehicle))
  (bind _skeletalmesh_2 (Variables|/Script/FortniteGame.FortAthenaVehicle|GetSkeletalMesh _vehicle_2))
  (bind (_returnvalue_location _returnvalue_rotation) (Math|Transform|BreakTransform (Transformation|GetSocketTransform _skeletalmesh_2 "gun_base" "RTS_Component")))
  (bind _turretrotation (Variables|Weapons|GetTurretRotation))
  (bind (_returnvalue_x_2 _returnvalue_y_2) (Math|Vector2D|BreakVector2D (Weapons|GetDeltaRotation _returnvalue_rotation _turretrotation 1)))
  (bind _returnvalue_2 (Movement|GetSpeedSign))
  (bind _speedparamvalue (Variables|Movement|GetSpeedParamValue))
  (bind _returnvalue_3 (Math|Float|Absolute(Float) _speedparamvalue))
  (bind _speedrangemax (Variables|Movement|GetSpeedRangeMax))
  (bind _speedrangemin (Variables|Movement|GetSpeedRangeMin))
  (bind _returnvalue_4 (Math|Float|NormalizetoRange _returnvalue_3 _speedrangemin _speedrangemax))
  (bind _righttrackpower (Variables|Movement|GetRightTrackPower))
  (bind _vehicle_3 (Variables|Pawn|GetVehicle))
  (bind _lefttrackpower (Variables|Movement|GetLeftTrackPower))
  (bind _engine (Variables|Default|GetEngine))
  (bind _engine_2 (Variables|Default|GetEngine))
  ;[1 1 1] On Tick - This will need to be nativized
  (if (not (AudioController|IsNativizationEnabled))
    ;;[1 0.7225 0.2854] Make sure vehicle is even valid
    (Variables|Pawn|GetVehicle
      (:then
        (if (Variables|Pawn|GetIsAwake)
          (Utilities|FlowControl|Sequence
            (:then_0
              ;;[1 1 1] Cache forward speed so we can update params
              (Variables|Movement|SetSpeedParamValue (Vehicle|GetForwardSpeedKmh (Variables|Pawn|GetVehicle)))
              ;;[1 1 1] Set turning parameters for Left and right, then update Turning Param for when not in motion
              ;;;[1 1 1] Cache power treads values
              (Variables|Movement|SetLeftTrackPower (TreadVehicle|GetTreadPower _vehicle_3 "Left" false))
              (Variables|Movement|SetRightTrackPower (TreadVehicle|GetTreadPower _vehicle_3 "Right" false))
              ;;;[1 1 1] Cache Turning value
              (Variables|Movement|SetTurningParamValue (Movement|UpdateTurningValue))
              (Utilities|FlowControl|Sequence
                (:then_0
                  ;;[0.7719 1 0] Accel and Decel Onshots
                  ;;;[1 1 1] Check if we are acceling forward/reverse.
                  (if (Movement|GetTryingMotion)
                    ;;;[1 1 1] Stop Decel audio if we are trying to be in motion
                    (bind _deceloneshotaudio (Variables|Default|GetDecelOneShotAudio)
                      (:then
                        (Audio|Components|Audio|FadeOut _deceloneshotaudio 0.25 0.0 "Linear" :!class /Script/Engine.AudioComponent))
                      (:else
                        _))
                    ;;;[1 1 1] Check the gear we are in before we are able to play accel audio. Also, check if speed is above "X" threshold before playing.
                    (if (and (!= _returnvalue_2 (Variables|Gearing|GetCurrentGear)) (> (Math|Float|Absolute(Float) (Variables|Movement|GetSpeedParamValue)) (Variables|Movement|GetAccelSpeedThreshold)))
                      (Variables|Gearing|SetCurrentGear _returnvalue_2)
                      ;;;[1 1 1] Make oneshot sounds wait for time to play again and make sure we are not in gear 0
                      (if (and (> (- (Utilities|Time|GetGameTimeinSeconds) (Variables|Gearing|GetLastGearChangeTime)) (Variables|Gearing|GetAccelWaitTime)) (!= (Variables|Gearing|GetCurrentGear) 0))
                        ;;;[1 1 1] Play Accel oneshot and set time until we can play again.
                        (CallFunction|PlayAccelOneShotBP)
                        (Variables|Gearing|SetLastGearChangeTime (Utilities|Time|GetGameTimeinSeconds))))
                    (else
                      ;;;[1 1 1] Check if we are trying motion against previous cached motion. If True update update last forward speed value.
                      (if (and (!= (Movement|GetTryingMotion) (Variables|Gearing|GetbPreviousTryMotion)) (Variables|Gearing|GetbPreviousTryMotion))
                        (Variables|Gearing|SetLastForwardSpeed (Math|Float|Absolute(Float) (Variables|Movement|GetSpeedParamValue)))
                        (else
                          ;;;[1 1 1] Fade out Accel one shot if it exists when we are not in Motion
                          (if (> (- (Utilities|Time|GetGameTimeinSeconds) (Variables|Gearing|GetLastGearChangeTime)) (Variables|Movement|GetTimeUntilAccelOneShotFade))
                            (bind _acceloneshotaudio (Variables|Default|GetAccelOneShotAudio)
                              (:then
                                (Audio|Components|Audio|FadeOut _acceloneshotaudio 0.25 0.0 "Linear" :!class /Script/Engine.AudioComponent))
                              (:else
                                _))
                            ;;;[1 1 1] Set Current gear to 0 and then see if speed from previous to play one shot audio when no longer in motion
                            (Variables|Gearing|SetCurrentGear 0)
                            (if (and (< (Variables|Movement|GetSpeedParamValue) (- (Variables|Gearing|GetLastForwardSpeed) (Variables|Gearing|GetDecelSpeedThreshhold))) (not (Movement|GetTurningMotion)))
                              ;;;[1 1 1] Play Decel audio once
                              (Utilities|FlowControl|DoOnce false
                                (:Completed
                                  (CallFunction|PlayDecelOneShotBP))))
                            (else
                              _)))))))
                (:then_1
                  ;;[1 1 1] Cache trying motion
                  (Variables|Gearing|SetbPreviousTryMotion (Movement|GetTryingMotion)))))
            (:then_1
              ;;[1 1 1] Play idle engine audio when vehicle is not in motion, otherwise fade out idle loop
              (Class|FortLayeredAudioComponent|SetFloatParameterExt (Variables|Default|GetEngine) "Idle" (Movement|MapRangeClampedwithBoost (+ _returnvalue_4 (Variables|Movement|GetTurningParamValue)) (Variables|Movement|GetEngineIdleThreshold)))
              ;;[1 1 1] Set overall speed parameter to maintain parity
              (Class|FortLayeredAudioComponent|SetFloatParameterExt (Variables|Default|GetEngine) "Speed" _returnvalue_4)
              ;;[1 1 1] Play engine forward audio
              (Class|FortLayeredAudioComponent|SetFloatParameterExt (Variables|Default|GetEngine) "Forward" (Movement|MapRangeClampedwithBoost (Variables|Movement|GetSpeedParamValue) (Variables|Movement|GetSpeedRangeMin) (Variables|Movement|GetSpeedRangeMax)))
              ;;[1 1 1] Play engine reverse audio
              (Class|FortLayeredAudioComponent|SetFloatParameterExt (Variables|Default|GetEngine) "Reverse" (Movement|MapRangeClampedwithBoost (Variables|Movement|GetSpeedParamValue) (Variables|Movement|GetSpeedRangeMin) (* (Variables|Movement|GetSpeedRangeMax) -1.0)))
              ;;[1 1 1] Play turning left and turning right audio, but not loud when performing other movements.
              (Class|FortLayeredAudioComponent|SetFloatParameterExt (Variables|Default|GetEngine) "Turning" (Movement|MapRangeClampedwithBoost (Variables|Movement|GetTurningParamValue) (Variables|Movement|GetTurningRangeMin) (Variables|Movement|GetTurningRangeMax)))
              (CallFunction|OnTurningAudio)
              ;;[1 1 1] Update Left and right tracks audio (This may change in the future. Just need to get the components attached to each side of vehicle properly.
              (Utilities|FlowControl|Sequence
                (:then_0
                  ;;;[1 1 1] Left Tracks
                  (Class|FortLayeredAudioComponent|SetFloatParameterExt (Variables|Default|GetTracksLeftAudio) "Forward" _lefttrackpower)
                  (Class|FortLayeredAudioComponent|SetFloatParameterExt (Variables|Default|GetTracksLeftAudio) "Reverse" (* _lefttrackpower -1.0)))
                (:then_1
                  ;;;[1 1 1] Right Tracks
                  (Class|FortLayeredAudioComponent|SetFloatParameterExt (Variables|Default|GetTracksRightAudio) "Forward" _righttrackpower)
                  (Class|FortLayeredAudioComponent|SetFloatParameterExt (Variables|Default|GetTracksRightAudio) "Reverse" (* _righttrackpower -1.0)))))
            (:then_2
              ;;[1 1 1] Weapons Movement - Get actor/mesh rotation to properly update pitch/yaw values
              ;;;[1 1 1] Cannon
              (Class|FortLayeredAudioComponent|SetFloatParameterExt _engine_2 "CannonHorizontalMovement" _returnvalue_x)
              (Class|FortLayeredAudioComponent|SetFloatParameterExt _engine_2 "CannonVerticalMovement" _returnvalue_y)
              (Variables|Weapons|SetCannonRotation _returnvalue)
              ;;;[1 1 1] Turret
              (Class|FortLayeredAudioComponent|SetFloatParameterExt _engine "TurretHorizontalMovement" _returnvalue_x_2)
              (Class|FortLayeredAudioComponent|SetFloatParameterExt _engine "TurretVerticalMovement" _returnvalue_y_2)
              (Variables|Weapons|SetTurretRotation _returnvalue_rotation))
            (:then_3
              ;;[1 1 1] Update surface audio that Vehicle is on
              (Audio|Parameter|SetFloatParameter (Variables|Default|GetEngine) "Surface" (Tracks|UpdateSurfaceType) :!class /Script/AudioExtensions.AudioParameterControllerInterface)
              ;;[1 1 1] Update is in water
              (Audio|Parameter|SetFloatParameter (Variables|Default|GetEngine) "IsInWater" (Math|Conversions|ToFloat(Boolean) (Variables|Movement|GetbIsInWater)) :!class /Script/AudioExtensions.AudioParameterControllerInterface)
              ;;[1 1 1] Has Fuel
              (Audio|Parameter|SetFloatParameter (Variables|Default|GetEngine) "OutOfFuel" (Math|Conversions|ToFloat(Boolean) (not (Variables|/Tank/Vehicle/Tank_Vehicle.Tank_Vehicle_C|GetVehicleHasFuel (Variables|Pawn|GetVehicle)))) :!class /Script/AudioExtensions.AudioParameterControllerInterface)
              ;;[1 1 1] IsInAir
              (Audio|Parameter|SetFloatParameter (Variables|Default|GetEngine) "IsInAir" (Math|Conversions|ToFloat(Boolean) (Vehicle|VehicleIsinAir (Variables|Pawn|GetVehicle))) :!class /Script/AudioExtensions.AudioParameterControllerInterface)))))
      (:else))))

(event Custom|OnTurningAudio
  (bind _returnvalue (Movement|GetTurningMotion))
  ;[1 1 1] On Tick - This will need to be nativized
  ;;[1 1 1] Updating Turning Oneshot for when turning left or turning right then play audio
  (Class|FortLayeredAudioComponent|SetFloatParameterExt (Variables|Default|GetEngine) "TurnPower" (Tracks|GetTreadVelocity))
  (if _returnvalue
    _
    (elif (!= _returnvalue true)
      (bind _turningoneshotaudio (Variables|Default|GetTurningOneShotAudio)
        (:then
          (Audio|Components|Audio|FadeOut _turningoneshotaudio (Variables|Sound|GetOneshotFadeOut) 0.0 "Linear" :!class /Script/Engine.AudioComponent))
        (:else
          _))))
  (Utilities|FlowControl|DoOnce false
    (:Completed
      (CallFunction|PlayTurningOneShotBP))))

(event Custom|DriverCoverActivateAudio
  (params (OnDriverEnter :type Boolean))
  ;[1 1 1] Driver Cover Audio
  (Audio|SpawnSoundAttached (Variables|Sound|GetDriverCoverActivatedSound) (Variables|/Script/FortniteGame.FortAthenaVehicle|GetSkeletalMesh (Variables|Pawn|GetVehicle)) "DriverCoverSocket_FX" "0, 0,50.000000"))

(event Custom|On%20Tick%20Stop%20Audio
  ;[1 1 1] On Tick Stop
  (Audio|Components|Audio|FadeOut (Variables|Default|GetEngine) 0.25 0.0 "Linear" :!class /Script/Engine.AudioComponent))

(event Custom|DriverCoverDeactivateAudio
  (params (OnDriverExit :type Boolean))
  ;[1 1 1] Driver Cover Audio
  (Audio|SpawnSoundAttached (Variables|Sound|GetDriverCoverDeactivatedSound) (Variables|/Script/FortniteGame.FortAthenaVehicle|GetSkeletalMesh (Variables|Pawn|GetVehicle)) "DriverCoverSocket_FX" "0, 0,50.000000"))

(event Custom|Cache%20Reference%20Vehicle
  (params (Vehicle :type Fort%20Tank%20Vehicle%20Object%20Reference))
  ;[1 1 1] Cach reference on vehicle begin play (Set engine component bAutoAttachManagement=false)
  (Utilities|IsValid Vehicle
    (:Is%20Valid
      (bind _astank_vehicle (Utilities|Casting|CastToTank_Vehicle Vehicle))
      (AudioController|SetVehicle _astank_vehicle)
      (Variables|Pawn|SetVehicle _astank_vehicle)
      (AudioController|SetLayeredAudioComponent (Variables|Default|GetEngine))
      (Tracks|AttachTracksAudio)
      (SurfaceType|BindEventtoOnSurfaceTypeVehicleIsonChangedDelegate (Variables|Pawn|GetVehicle) (EventDispatchers|CreateEvent :!event GetVehicleSurfaceType)))))

(event Custom|TreadsRepairAudio
  (params (BoneName :type Name))
  ;[1 1 1] Repair Audio
  (Audio|PlaySoundatLocation (Variables|Sound|GetTankTreadRepairedSound) (Transformation|GetActorLocation)))

(event Custom|On%20Sleep%20State%20Change
  (params (Awake :type Boolean))
  ;[1 1 1] Vehicle state when sleep or awake - When player is not in vehicle it should be in sleep mode
  (bind _output_get (Variables|Pawn|SetIsAwake Awake))
  (Utilities|Struct|SetmembersinTankAudioUpdateContext (Variables|FortTankVehicleAudioController|GetContext) _output_get))

(event Custom|On%20Collision%20Hit%20
  (params (HitPawn :type Actor%20Object%20Reference) (ImpactPoint :type Vector))
  (bind (_asfort_pawn _bsuccess) (Utilities|Casting|CastToFortPawn HitPawn))
  ;[1 1 1] Collision
  (Utilities|Casting|CastToLandscape HitPawn
    (:then
      (if (Variables|Damage|GetbIsDestroyed)
        (Damage|DestroyedLandAudio)))
    (:CastFailed
      (Audio|OnCollision (Variables|Default|GetCollisionAudio_Hit) ImpactPoint (Math|Conversions|ToVector(Float) (Math|Float|SelectFloat 1.0 (Math|Float|MapRangeClamped (Math|Float|Absolute(Float) (Variables|Movement|GetSpeedParamValue)) 0.0 6e+01 0.0 1.0) _bsuccess))))))

(event Custom|On%20Driver%20State%20Change%20Audio
  (params (HasDriver :type Boolean))
  ;[1 1 1] When vehicle has driver set float to 1, crank vehicle
  (Variables|Pawn|SetbHasDriver HasDriver)
  (Audio|Parameter|SetFloatParameter (Variables|Default|GetEngine) "HasDriver" (Math|Conversions|ToFloat(Boolean) (Variables|Pawn|GetbHasDriver)) :!class /Script/AudioExtensions.AudioParameterControllerInterface)
  (Audio|Parameter|SetIntegerParameter (Variables|Default|GetEngineOnOff) "OutOfFeul" (Math|Conversions|ToInteger(Boolean) (not (Variables|/Tank/Vehicle/Tank_Vehicle.Tank_Vehicle_C|GetVehicleHasFuel (Variables|Pawn|GetVehicle)))) :!class /Script/AudioExtensions.AudioParameterControllerInterface)
  (Audio|Parameter|SetIntegerParameter (Variables|Default|GetEngineOnOff) "EngineOn" (Math|Conversions|ToInteger(Boolean) (Variables|Pawn|GetbHasDriver)) :!class /Script/AudioExtensions.AudioParameterControllerInterface)
  (Audio|Components|Audio|Play (Variables|Default|GetEngineOnOff) 0.0)
  (CallFunction|UpdateLocalOccupancy)
  (Tracks|UpdateTracksStatus))

(event Custom|OnPassengerStateChangeAudio
  (params (IsLocalPlayerPassenger :type Boolean))
  (bind _vehicle (Variables|Pawn|GetVehicle))
  (bind _returnvalue (Vehicle|GetPawnatSeat _vehicle 1))
  (bind _engine (Variables|Default|GetEngine))
  ;[1 1 1] Set local param for when passengers enter
  (CallFunction|UpdateLocalOccupancy)
  (Utilities|IsValid _returnvalue
    (:Is%20Valid
      (Audio|Parameter|SetBooleanParameter _engine "IsLocalTurret" (Pawn|IsLocallyViewed _returnvalue) :!class /Script/AudioExtensions.AudioParameterControllerInterface))
    (:Is%20Not%20Valid
      (Audio|Parameter|SetBooleanParameter _engine "IsLocalTurret" false :!class /Script/AudioExtensions.AudioParameterControllerInterface))))

(event Custom|TracksTirePopAudio
  (params (bRightTrack :type Boolean))
  ;[1 1 1] Tracks Audio Pop event
  (Audio|SpawnSoundAttached (Variables|Sound|GetTracksTirePopSound) (Variables|/Script/FortniteGame.FortAthenaVehicle|GetSkeletalMesh (Variables|Pawn|GetVehicle)) (select bRightTrack (Variables|Tracks|GetRightTracks) (Variables|Tracks|GetLeftTracks)) (select bRightTrack "0,65.000000, 0" "0,-65.000000, 0"))
  (Tracks|UpdateTracksStatus))

(event Custom|On%20Water%20State%20Change%20Audio
  (params (IsEnteringWater :type Boolean))
  ;[1 1 1] Water state change - this may not be working for tank just yet. Duped from valet
  (if (Math|Boolean|NotEqual(Boolean) IsEnteringWater (Variables|Movement|GetbIsInWater))
    (Variables|Movement|SetbIsInWater IsEnteringWater)
    (Audio|PlaySoundatLocation (select (Variables|Movement|GetbIsInWater) (Variables|Sound|GetWaterEnterSoundCue) (Variables|Sound|GetWaterExitSoundCue)) (Transformation|GetWorldLocation (Variables|Default|GetEngine)))))

(event Custom|On%20Damage%20Audio
  (params (Damage :type Float%20%28double-precision%29))
  ;[1 1 1] Tank Health Damaged States
  (Variables|Damage|SetDamageValue Damage)
  (Variables|Pawn|GetVehicle
    (:then
      (if (not (Variables|Damage|GetbIsDestroyed))
        (Utilities|FlowControl|Sequence
          (:then_0
            ;;[1 1 1] Stage one damage loop
            (if (>= (Variables|Damage|GetDamageValue) (Variables|/Tank/Vehicle/Tank_Vehicle.Tank_Vehicle_C|GetDamage_Stage1 (Variables|Pawn|GetVehicle)))
              (Variables|Default|GetFirstDamageStateAudio
                (:then)
                (:else
                  (bind _returnvalue (Audio|SpawnSoundAttached (Variables|Sound|GetFirstDamageStateSound) (Variables|Default|GetEngine)))
                  (Variables|Default|SetFirstDamageStateAudio _returnvalue)))
              (else
                (bind _firstdamagestateaudio (Variables|Default|GetFirstDamageStateAudio)
                  (:then
                    (Audio|Components|Audio|FadeOut _firstdamagestateaudio 1.5 0.0 "Linear" :!class /Script/Engine.AudioComponent))
                  (:else)))))
          (:then_1
            ;;[1 1 1] Second stage damage loop, but not sure if we need this in current setup. Need content for it.
            (if (>= (Variables|Damage|GetDamageValue) (Variables|/Tank/Vehicle/Tank_Vehicle.Tank_Vehicle_C|GetDamage_Stage2 (Variables|Pawn|GetVehicle)))
              (Variables|Default|GetSecondDamageStateAudio
                (:then)
                (:else
                  (bind _returnvalue_2 (Audio|SpawnSoundAttached (Variables|Sound|GetSecondDamageStateSound) (Variables|Default|GetEngine)))
                  (Variables|Default|SetSecondDamageStateAudio _returnvalue_2)))
              (else
                (bind _seconddamagestateaudio (Variables|Default|GetSecondDamageStateAudio)
                  (:then
                    (Audio|Components|Audio|FadeOut _seconddamagestateaudio 1.5 0.0 "Linear" :!class /Script/Engine.AudioComponent))
                  (:else)))))
          (:then_2
            ;;[1 1 1] Final Damage Stage loop
            (if (>= (Variables|Damage|GetDamageValue) (Variables|/Tank/Vehicle/Tank_Vehicle.Tank_Vehicle_C|GetDamage_Stage3Critical (Variables|Pawn|GetVehicle)))
              (Variables|Default|GetPreDestroyAudio
                (:then)
                (:else
                  (bind _returnvalue_3 (Audio|SpawnSoundAttached (Variables|Sound|GetDamagePreDestroySound) (Variables|Default|GetEngine)))
                  (Variables|Default|SetPreDestroyAudio _returnvalue_3)))
              (else
                (bind _predestroyaudio (Variables|Default|GetPreDestroyAudio)
                  (:then
                    (Audio|Components|Audio|FadeOut _predestroyaudio 1.5 0.0 "Linear" :!class /Script/Engine.AudioComponent))
                  (:else))))))))
    (:else)))

(event Custom|On%20Horn%20Audio
  (bind _horn (Variables|Default|GetHorn))
  ;[1 1 1] On Horn
  (if (not (Audio|Components|Audio|IsPlaying _horn))
    (Audio|Components|Audio|Play _horn 0.0)))

(event Custom|On%20Destroy%20Audio
  ;[1 1 1] Destroyed Event - Once Tank has blown up stop all audio that is not related
  (Damage|StopDamageAudio)
  (Audio|Components|Audio|Stop (Variables|Default|GetEngine))
  (bind _returnvalue (Audio|SpawnSoundAttached (Variables|Sound|GetDestroySound) (Variables|Default|GetEngine)))
  (Variables|Default|SetDestroyedAudio _returnvalue)
  (Variables|Damage|SetbIsDestroyed true)
  (Utilities|FlowControl|Sequence
    (:then_0
      (bind _returnvalue_2 (Utilities|Time|SetTimerbyEvent (EventDispatchers|CreateEvent :!event Custom|Destroyed%20Launch%20Audio) 0.5))
      (Variables|Damage|SetLaunchUpOverlapHandle _returnvalue_2))
    (:then_1
      (Utilities|FlowControl|Delay 3.0)
      (bind _returnvalue_3 (Utilities|Time|SetTimerbyEvent (EventDispatchers|CreateEvent :!event Custom|On%20Destroyed%20Launch) 0.1 true))
      (Variables|Damage|SetLaunchedOverlapHandle _returnvalue_3))))

(event Custom|On%20Destroyed%20Launch
  ;[1 1 1] Destroyed Event - Once Tank has blown up stop all audio that is not related
  (bind _vehicle (Variables|Pawn|GetVehicle)
    (:then
      (bind _outactors (Collision|SphereOverlapActors (Transformation|GetActorLocation _vehicle) 4.5e+03 (Variables|Damage|GetObject%20Types) "/Script/FortniteGame.FortPlayerPawn"))
      (for _ _outactors
        (Utilities|FlowControl|DoOnce false
          (:Completed
            (Damage|OnLaunchPassby)))))
    (:else)))

(event Custom|Boost%20Start%20Audio
  ;[1 1 1] Boost
  (bind _returnvalue (Audio|SpawnSoundAttached (Variables|Sound|GetBoostSound) (Variables|Default|GetEngine)))
  (Variables|Default|SetBoostAccelAudio _returnvalue)
  (bind _boostdecelaudio (Variables|Default|GetBoostDecelAudio)
    (:then
      (Audio|Components|Audio|FadeOut _boostdecelaudio 0.5 0.0 "Linear" :!class /Script/Engine.AudioComponent)
      (OneShots|StopOneShots))
    (:else)))

(event Custom|Destroyed%20Launch%20Audio
  ;[1 1 1] Destroyed Event - Once Tank has blown up stop all audio that is not related
  (Damage|OnLaunchPassby))

(event Custom|Boost%20End%20Audio
  ;[1 1 1] Boost
  (bind _returnvalue (Audio|SpawnSoundAttached (Variables|Sound|GetBoostEndSound) (Variables|Default|GetEngine)))
  (Variables|Default|SetBoostDecelAudio _returnvalue)
  (CallFunction|StopBoostAudio))

(event Custom|Vehicle%20End%20Play%20Audio
  ;[1 1 1] Stop any looping audio that is still happening
  (Damage|StopDamageAudio)
  (Utilities|FlowControl|Sequence
    (:then_0
      (bind _destroyedaudio (Variables|Default|GetDestroyedAudio)
        (:then
          (Audio|Components|Audio|FadeOut _destroyedaudio 1.5 0.0 "Linear" :!class /Script/Engine.AudioComponent))
        (:else)))
    (:then_1
      (Utilities|Time|ClearandInvalidateTimerbyHandle (Variables|Damage|GetLaunchedOverlapHandle))
      (Utilities|Time|ClearandInvalidateTimerbyHandle (Variables|Damage|GetLaunchUpOverlapHandle))
      (bind _launchpassbyaudio (Variables|Default|GetLaunchPassbyAudio)
        (:then
          (Audio|Components|Audio|FadeOut _launchpassbyaudio 0.5 0.0 "Linear" :!class /Script/Engine.AudioComponent))
        (:else)))))

(event Custom|StopBoostAudio
  ;[1 1 1] Boost
  (bind _boostaccelaudio (Variables|Default|GetBoostAccelAudio)
    (:then
      (Audio|Components|Audio|FadeOut _boostaccelaudio 0.5 0.0 "Linear" :!class /Script/Engine.AudioComponent))
    (:else)))

(event Custom|EngineDamageAudio
  (params (HealthPercent :type Float%20%28double-precision%29))
  ;[1 1 1] Damage Engine audio
  (if (<= HealthPercent (Variables|/Tank/Vehicle/Tank_Vehicle.Tank_Vehicle_C|GetEngineHealthPercentToEnableEffects (Variables|Pawn|GetVehicle)))
    (Variables|Default|GetEngineDamageLoopAudio
      (:then)
      (:else
        (bind _returnvalue (Audio|SpawnSoundAttached (Variables|Sound|GetEngineDamagedSound) (Variables|Default|GetEngineOnOff)))
        (Variables|Default|SetEngineDamageLoopAudio _returnvalue)))
    (else
      (bind _enginedamageloopaudio (Variables|Default|GetEngineDamageLoopAudio)
        (:then
          (Audio|Components|Audio|FadeOut _enginedamageloopaudio 1.0 0.0 "Linear" :!class /Script/Engine.AudioComponent))
        (:else)))))

(event Custom|EngineOverheatedAudio
  ;[1 1 1] Tank overheated/disabled
  (if (Variables|/Tank/Vehicle/Tank_Vehicle.Tank_Vehicle_C|GetEngineOverheated (Variables|Pawn|GetVehicle))
    (Variables|Default|GetEngineDisabledAudio
      (:then)
      (:else
        (bind _returnvalue (Audio|SpawnSoundAttached (Variables|Sound|GetDisabledSound) (Variables|Default|GetEngineOnOff)))
        (Variables|Default|SetEngineDisabledAudio _returnvalue)))
    (else
      (bind _enginedisabledaudio (Variables|Default|GetEngineDisabledAudio)
        (:then
          (Audio|Components|Audio|FadeOut _enginedisabledaudio 1.0 0.0 "Linear" :!class /Script/Engine.AudioComponent)
          (Audio|SpawnSoundAttached (Variables|Sound|GetTankEnginePowerUpSound) (Variables|Default|GetEngineOnOff)))
        (:else)))))

(event EventPlayAccelOneShot
  ;[1 1 1] One-shot playback
  (Utilities|FlowControl|Sequence
    (:then_0
      (bind _acceloneshotaudio (Variables|Default|GetAccelOneShotAudio)
        (:then
          (Audio|Components|Audio|FadeOut _acceloneshotaudio (Variables|Sound|GetOneshotFadeOut) 0.0 "Linear" :!class /Script/Engine.AudioComponent))
        (:else
          _))
      (bind _returnvalue (Audio|SpawnSoundAttached (Variables|Sound|GetAccelOneShotSound) (Variables|Default|GetEngine)))
      (bind _output_get (Variables|Default|SetAccelOneShotAudio _returnvalue))
      (OneShots|PlayGearShift)
      (Utilities|Struct|SetmembersinTankAudioUpdateContext (Variables|FortTankVehicleAudioController|GetContext) _output_get))
    (:then_1
      (bind _deceloneshotaudio (Variables|Default|GetDecelOneShotAudio)
        (:then
          (Audio|Components|Audio|FadeOut _deceloneshotaudio (Variables|Sound|GetOneshotFadeOut) 0.0 "Linear" :!class /Script/Engine.AudioComponent))
        (:else
          _))
      (bind _turningoneshotaudio (Variables|Default|GetTurningOneShotAudio)
        (:then
          (Audio|Components|Audio|FadeOut _turningoneshotaudio (Variables|Sound|GetOneshotFadeOut) 0.0 "Linear" :!class /Script/Engine.AudioComponent))
        (:else)))))

(event Custom|PlayAccelOneShotBP
  ;[1 1 1] One-shot playback
  (Utilities|FlowControl|Sequence
    (:then_0
      (bind _acceloneshotaudio (Variables|Default|GetAccelOneShotAudio)
        (:then
          (Audio|Components|Audio|FadeOut _acceloneshotaudio (Variables|Sound|GetOneshotFadeOut) 0.0 "Linear" :!class /Script/Engine.AudioComponent))
        (:else
          _))
      (bind _returnvalue (Audio|SpawnSoundAttached (Variables|Sound|GetAccelOneShotSound) (Variables|Default|GetEngine)))
      (bind _output_get (Variables|Default|SetAccelOneShotAudio _returnvalue))
      (OneShots|PlayGearShift)
      (Utilities|Struct|SetmembersinTankAudioUpdateContext (Variables|FortTankVehicleAudioController|GetContext) _output_get))
    (:then_1
      (bind _deceloneshotaudio (Variables|Default|GetDecelOneShotAudio)
        (:then
          (Audio|Components|Audio|FadeOut _deceloneshotaudio (Variables|Sound|GetOneshotFadeOut) 0.0 "Linear" :!class /Script/Engine.AudioComponent))
        (:else
          _))
      (bind _turningoneshotaudio (Variables|Default|GetTurningOneShotAudio)
        (:then
          (Audio|Components|Audio|FadeOut _turningoneshotaudio (Variables|Sound|GetOneshotFadeOut) 0.0 "Linear" :!class /Script/Engine.AudioComponent))
        (:else)))))

(event EventPlayDecelOneShot
  ;[1 1 1] One-shot playback
  (Utilities|FlowControl|Sequence
    (:then_0
      (bind _deceloneshotaudio (Variables|Default|GetDecelOneShotAudio)
        (:then
          (Audio|Components|Audio|FadeOut _deceloneshotaudio (Variables|Sound|GetOneshotFadeOut) 0.0 "Linear" :!class /Script/Engine.AudioComponent))
        (:else
          _))
      (bind _returnvalue (Audio|SpawnSoundAttached (Variables|Sound|GetDecelOneShotSound) (Variables|Default|GetEngine)))
      (bind _output_get (Variables|Default|SetDecelOneShotAudio _returnvalue))
      (OneShots|PlayGearShift)
      (Utilities|Struct|SetmembersinTankAudioUpdateContext (Variables|FortTankVehicleAudioController|GetContext) _output_get))
    (:then_1
      (bind _acceloneshotaudio (Variables|Default|GetAccelOneShotAudio)
        (:then
          (Audio|Components|Audio|FadeOut _acceloneshotaudio (Variables|Sound|GetOneshotFadeOut) 0.0 "Linear" :!class /Script/Engine.AudioComponent))
        (:else
          _))
      (bind _turningoneshotaudio (Variables|Default|GetTurningOneShotAudio)
        (:then
          (Audio|Components|Audio|FadeOut _turningoneshotaudio (Variables|Sound|GetOneshotFadeOut) 0.0 "Linear" :!class /Script/Engine.AudioComponent))
        (:else)))))

(event Custom|PlayDecelOneShotBP
  ;[1 1 1] One-shot playback
  (Utilities|FlowControl|Sequence
    (:then_0
      (bind _deceloneshotaudio (Variables|Default|GetDecelOneShotAudio)
        (:then
          (Audio|Components|Audio|FadeOut _deceloneshotaudio (Variables|Sound|GetOneshotFadeOut) 0.0 "Linear" :!class /Script/Engine.AudioComponent))
        (:else
          _))
      (bind _returnvalue (Audio|SpawnSoundAttached (Variables|Sound|GetDecelOneShotSound) (Variables|Default|GetEngine)))
      (bind _output_get (Variables|Default|SetDecelOneShotAudio _returnvalue))
      (OneShots|PlayGearShift)
      (Utilities|Struct|SetmembersinTankAudioUpdateContext (Variables|FortTankVehicleAudioController|GetContext) _output_get))
    (:then_1
      (bind _acceloneshotaudio (Variables|Default|GetAccelOneShotAudio)
        (:then
          (Audio|Components|Audio|FadeOut _acceloneshotaudio (Variables|Sound|GetOneshotFadeOut) 0.0 "Linear" :!class /Script/Engine.AudioComponent))
        (:else
          _))
      (bind _turningoneshotaudio (Variables|Default|GetTurningOneShotAudio)
        (:then
          (Audio|Components|Audio|FadeOut _turningoneshotaudio (Variables|Sound|GetOneshotFadeOut) 0.0 "Linear" :!class /Script/Engine.AudioComponent))
        (:else)))))

(event EventPlayTurningOneShot
  ;[1 1 1] One-shot playback
  (Utilities|FlowControl|Sequence
    (:then_0
      (bind _turningoneshotaudio (Variables|Default|GetTurningOneShotAudio)
        (:then
          (Audio|Components|Audio|FadeOut _turningoneshotaudio (Variables|Sound|GetOneshotFadeOut) 0.0 "Linear" :!class /Script/Engine.AudioComponent))
        (:else
          _))
      (bind _returnvalue (Audio|SpawnSoundAttached (Variables|Sound|GetTurningOneShotSound) (Variables|Default|GetEngine)))
      (bind _output_get (Variables|Default|SetTurningOneShotAudio _returnvalue))
      (Utilities|Struct|SetmembersinTankAudioUpdateContext (Variables|FortTankVehicleAudioController|GetContext) _output_get))
    (:then_1
      (bind _acceloneshotaudio (Variables|Default|GetAccelOneShotAudio)
        (:then
          (Audio|Components|Audio|FadeOut _acceloneshotaudio (Variables|Sound|GetOneshotFadeOut) 0.0 "Linear" :!class /Script/Engine.AudioComponent))
        (:else
          _))
      (bind _deceloneshotaudio (Variables|Default|GetDecelOneShotAudio)
        (:then
          (Audio|Components|Audio|FadeOut _deceloneshotaudio (Variables|Sound|GetOneshotFadeOut) 0.0 "Linear" :!class /Script/Engine.AudioComponent))
        (:else)))))

(event Custom|PlayTurningOneShotBP
  ;[1 1 1] One-shot playback
  (Utilities|FlowControl|Sequence
    (:then_0
      (bind _turningoneshotaudio (Variables|Default|GetTurningOneShotAudio)
        (:then
          (Audio|Components|Audio|FadeOut _turningoneshotaudio (Variables|Sound|GetOneshotFadeOut) 0.0 "Linear" :!class /Script/Engine.AudioComponent))
        (:else
          _))
      (bind _returnvalue (Audio|SpawnSoundAttached (Variables|Sound|GetTurningOneShotSound) (Variables|Default|GetEngine)))
      (bind _output_get (Variables|Default|SetTurningOneShotAudio _returnvalue))
      (Utilities|Struct|SetmembersinTankAudioUpdateContext (Variables|FortTankVehicleAudioController|GetContext) _output_get))
    (:then_1
      (bind _acceloneshotaudio (Variables|Default|GetAccelOneShotAudio)
        (:then
          (Audio|Components|Audio|FadeOut _acceloneshotaudio (Variables|Sound|GetOneshotFadeOut) 0.0 "Linear" :!class /Script/Engine.AudioComponent))
        (:else
          _))
      (bind _deceloneshotaudio (Variables|Default|GetDecelOneShotAudio)
        (:then
          (Audio|Components|Audio|FadeOut _deceloneshotaudio (Variables|Sound|GetOneshotFadeOut) 0.0 "Linear" :!class /Script/Engine.AudioComponent))
        (:else)))))

(fn AttachTracksAudio
  (Utilities|FlowControl|Sequence
    (:then_0
      ;[1 1 1] Right Track
      (Transformation|AttachComponentToComponent (Variables|Default|GetTracksRightAudio) (Variables|/Script/FortniteGame.FortAthenaVehicle|GetSkeletalMesh (Variables|Pawn|GetVehicle)) (Variables|Tracks|GetRightTracks) "SnapToTarget" "KeepWorld" "SnapToTarget" true)
      (Transformation|AddLocalOffset (Variables|Default|GetTracksRightAudio) "0.000000,65.000000,0.000000" false false))
    (:then_1
      ;[1 1 1] Left Track
      (Transformation|AttachComponentToComponent (Variables|Default|GetTracksLeftAudio) (Variables|/Script/FortniteGame.FortAthenaVehicle|GetSkeletalMesh (Variables|Pawn|GetVehicle)) (Variables|Tracks|GetLeftTracks) "SnapToTarget" "KeepWorld" "SnapToTarget" true)
      (Transformation|AddLocalOffset (Variables|Default|GetTracksLeftAudio) "0.000000,-65.000000,0.000000" false false))))

(fn CompressionTracksAudio
  (params (HitResult :type Hit%20Result%20Structure) (Compression :type Float%20%28double-precision%29) (CompressionVelocity :type Float%20%28double-precision%29))
  (bind (_bblockinghit _binitialoverlap _time _distance _location _impactpoint) (Collision|BreakHitResult HitResult))
  ;[1 1 1] Play  compression audio
  (Audio|OnCollision (Variables|Default|GetCollisionAudio_Suspension) _impactpoint (Math|Vector|MakeVector CompressionVelocity)))

(fn DestroyedLandAudio
  (Utilities|FlowControl|Sequence
    (:then_0
      (Variables|Default|GetDestroyedAudio
        (:then
          (Audio|PlaySoundatLocation (Variables|Sound|GetDestroyedLandSound) (Transformation|GetActorLocation (Variables|Pawn|GetVehicle))))
        (:else)))
    (:then_1
      (Utilities|Time|ClearandInvalidateTimerbyHandle (Variables|Damage|GetLaunchedOverlapHandle)))))

(fn GetDeltaRotation
  (params (ObjectRotation :type Rotator) (CachedRotation :type Rotator) (SeatIndex :type Integer))
  (bind (_returnvalue_roll _returnvalue_pitch _returnvalue_yaw) (Math|Rotator|BreakRotator (Math|Rotator|Delta(Rotator) ObjectRotation CachedRotation)))
  (return (select (Utilities|IsValid (Vehicle|GetPawnatSeat (Variables|Pawn|GetVehicle) SeatIndex) :!class /Script/Engine.KismetSystemLibrary) (Math|Vector2D|MakeVector2D (Math|Float|MapRangeClamped (Math|Float|Absolute(Float) _returnvalue_yaw) (Variables|Weapons|GetWeaponsMovementMinThreshold) (Variables|Weapons|GetWeaponsMovementMaxThreshold) 0.0 1.0) (Math|Float|MapRangeClamped (Math|Float|Absolute(Float) _returnvalue_pitch) (Variables|Weapons|GetWeaponsMovementMinThreshold) (Variables|Weapons|GetWeaponsMovementMaxThreshold) 0.0 1.0)) 0)))

(fn GetTryingMotion
  (return (or (Vehicle|PlayerIsTryingtoMoveForward (Variables|Pawn|GetVehicle)) (Vehicle|PlayerIsTryingtoReverse (Variables|Pawn|GetVehicle)))))

(fn GetSpeedSign
  ;[1 1 1] Get speed in kmh that is rounded to make sure we hit zero properly. Values will output -1, 0, 1.
  (return (Math|Float|Truncate (Math|Float|Sign(Float) (DisplayHelper|RoundFloat (Vehicle|GetForwardSpeedKmh (Variables|Pawn|GetVehicle)) "ToZero")))))

(fn GetTreadVelocity
  (return (Math|Float|MapRangeClamped (Math|Float|Max(Float) (TreadVehicle|GetTreadVelocity (Variables|Pawn|GetVehicle) "Left" "KilometersPerHour") (TreadVehicle|GetTreadVelocity (Variables|Pawn|GetVehicle) "Right" "KilometersPerHour")) (Variables|Tracks|GetMinTreadVelocity) (Variables|Tracks|GetMaxTreadVelocity) 0.0 1.0)))

(fn GetTurningMotion
  (return (or (Variables|Movement|GetbTurningRight) (Variables|Movement|GetbTurningLeft))))

(fn GetVehicleSurfaceType
  (params (SurfaceTypeVehicleIsOn :type EPhysicalSurface%20Enum :default "SurfaceType_Default"))
  (bind (_value _returnvalue) (Utilities|Map|Find (Variables|Tracks|GetSurfaceTypeMap) (Variables|Tracks|GetSurface%20Type%20Vehicle%20Is%20On)))
  (Variables|Tracks|SetSurface%20Type%20Vehicle%20Is%20On SurfaceTypeVehicleIsOn)
  (if (or (Vehicle|VehicleIsinAir (Variables|Pawn|GetVehicle)) (Class|FortVehicleInterface|IsinWater (Variables|Pawn|GetVehicle)))
    (Variables|Tracks|SetSurfaceID 0)
    (Utilities|Struct|SetmembersinTankAudioUpdateContext (Variables|FortTankVehicleAudioController|GetContext) 0)
    (else
      (bind _output_get (Variables|Tracks|SetSurfaceID (Math|Integer|SelectInt _value 1 _returnvalue)))
      (Utilities|Struct|SetmembersinTankAudioUpdateContext (Variables|FortTankVehicleAudioController|GetContext) _output_get)
      (bind _skidaudio (Variables|Default|GetSkidAudio)
        (:then
          (Audio|Components|Audio|FadeOut _skidaudio 0.25 0.0 "Linear" :!class /Script/Engine.AudioComponent))
        (:else
          _))
      (bind _returnvalue_2 (Audio|SpawnSoundAttached (Variables|Sound|GetSkidSound) (Variables|Default|GetEngine)))
      (bind _output_get_2 (Variables|Default|SetSkidAudio _returnvalue_2))
      (Audio|Parameter|SetIntegerParameter _output_get_2 "Skid" _output_get :!class /Script/AudioExtensions.AudioParameterControllerInterface))))

(fn MapRangeClampedWithBoost
  (params (Value :type Float%20%28double-precision%29) (InRangeA :type Float%20%28double-precision%29) (InRangeB :type Float%20%28double-precision%29))
  (return (Math|Float|SelectFloat 0.0 (Math|Float|MapRangeClamped Value InRangeA InRangeB 0.0 1.0) (Class|FortVehicleInterface|IsBoosting (Variables|Pawn|GetVehicle)))))

(fn OnLaunchPassby
  (bind _returnvalue (Audio|SpawnSoundAttached (Variables|Sound|GetDestroyedPassBySound) (Variables|Default|GetEngine)))
  (Variables|Default|SetLaunchPassbyAudio _returnvalue))

(fn PlayGearShift
  (bind _gearshiftaudio (Variables|Default|GetGearShiftAudio)
    (:then
      (Audio|Components|Audio|FadeOut _gearshiftaudio 0.25 0.0 "Linear" :!class /Script/Engine.AudioComponent))
    (:else
      _))
  (bind _returnvalue (Audio|SpawnSoundAttached (Variables|Sound|GetGearShiftSound) (Variables|Default|GetEngine)))
  (Variables|Default|SetGearShiftAudio _returnvalue))

(fn StopDamageAudio
  (Utilities|FlowControl|Sequence
    (:then_0
      ;[1 1 1] Body
      (bind _firstdamagestateaudio (Variables|Default|GetFirstDamageStateAudio)
        (:then
          (Audio|Components|Audio|FadeOut _firstdamagestateaudio 1.5 0.0 "Linear" :!class /Script/Engine.AudioComponent))
        (:else
          _))
      (bind _seconddamagestateaudio (Variables|Default|GetSecondDamageStateAudio)
        (:then
          (Audio|Components|Audio|FadeOut _seconddamagestateaudio 1.5 0.0 "Linear" :!class /Script/Engine.AudioComponent))
        (:else
          _))
      (bind _predestroyaudio (Variables|Default|GetPreDestroyAudio)
        (:then
          (Audio|Components|Audio|FadeOut _predestroyaudio 1.5 0.0 "Linear" :!class /Script/Engine.AudioComponent))
        (:else
          _))
      (bind _enginedamageloopaudio (Variables|Default|GetEngineDamageLoopAudio)
        (:then
          (Audio|Components|Audio|FadeOut _enginedamageloopaudio 1.5 0.0 "Linear" :!class /Script/Engine.AudioComponent))
        (:else)))
    (:then_1
      ;[1 1 1] Engine
      (bind _enginedisabledaudio (Variables|Default|GetEngineDisabledAudio)
        (:then
          (Audio|Components|Audio|FadeOut _enginedisabledaudio 1.5 0.0 "Linear" :!class /Script/Engine.AudioComponent))
        (:else
          _))
      (bind _enginedamageloopaudio_2 (Variables|Default|GetEngineDamageLoopAudio)
        (:then
          (Audio|Components|Audio|FadeOut _enginedamageloopaudio_2 1.5 0.0 "Linear" :!class /Script/Engine.AudioComponent))
        (:else)))))

(fn StopOneShots
  (bind _acceloneshotaudio (Variables|Default|GetAccelOneShotAudio)
    (:then
      (Audio|Components|Audio|FadeOut _acceloneshotaudio 0.25 0.0 "Linear" :!class /Script/Engine.AudioComponent))
    (:else
      _))
  (bind _deceloneshotaudio (Variables|Default|GetDecelOneShotAudio)
    (:then
      (Audio|Components|Audio|FadeOut _deceloneshotaudio 0.25 0.0 "Linear" :!class /Script/Engine.AudioComponent))
    (:else
      _))
  (bind _turningoneshotaudio (Variables|Default|GetTurningOneShotAudio)
    (:then
      (Audio|Components|Audio|FadeOut _turningoneshotaudio 0.25 0.0 "Linear" :!class /Script/Engine.AudioComponent))
    (:else)))

(fn UpdateLocalOccupancy
  (Audio|Parameter|SetBooleanParameter (AudioController|GetAudioParameterComponent) "IsLocalPassenger" (or (Variables|/Tank/Vehicle/Tank_Vehicle.Tank_Vehicle_C|GetIsLocalPlayerPassenger (Variables|Pawn|GetVehicle)) (Vehicle|ClientIsDriver (Variables|Pawn|GetVehicle))) :!class /Script/AudioExtensions.AudioParameterControllerInterface))

(fn UpdateSurfaceType
  (bind _output_get (Variables|Tracks|SetCurrentSurfaceTypeParam (Math|Interpolation|FInterpTo (Variables|Tracks|GetCurrentSurfaceTypeParam) (/ (Math|Conversions|ToFloat(Integer) (Variables|Tracks|GetSurfaceID)) 3.0) (Utilities|Time|GetWorldDeltaSeconds) (Variables|Tracks|GetSurfaceInterpSpeed))))
  (return _output_get))

(fn UpdateTracksStatus
  (Audio|Parameter|SetBooleanParameter (Variables|Default|GetTracksLeftAudio) "IsEnabled" (TreadVehicle|IsLeftTreadEnabled (Variables|Pawn|GetVehicle)) :!class /Script/AudioExtensions.AudioParameterControllerInterface)
  (Audio|Parameter|SetBooleanParameter (Variables|Default|GetTracksRightAudio) "IsEnabled" (TreadVehicle|IsRightTreadEnabled (Variables|Pawn|GetVehicle)) :!class /Script/AudioExtensions.AudioParameterControllerInterface))

(fn UpdateTurningValue
  ;[1 1 1] Turning Right set bool
  (Variables|Movement|SetbTurningRight (and (== (Variables|Movement|GetLeftTrackPower) 1.0) (== (Variables|Movement|GetRightTrackPower) -1.0)))
  ;[1 1 1] Turning Left set bool
  (Variables|Movement|SetbTurningLeft (and (== (Variables|Movement|GetLeftTrackPower) -1.0) (== (Variables|Movement|GetRightTrackPower) 1.0)))
  (return (Math|Float|SelectFloat 0.0 (Math|Float|SelectFloat (Math|Float|Absolute(Float) (Variables|Movement|GetLeftTrackPower)) (Math|Float|Absolute(Float) (Variables|Movement|GetRightTrackPower)) (> (Math|Float|Absolute(Float) (Variables|Movement|GetLeftTrackPower)) (Math|Float|Absolute(Float) (Variables|Movement|GetRightTrackPower)))) (Movement|GetTryingMotion))))

(fn ConstructionScript)
